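; GA_InVortex is an event-triggered gameplay ability that puts the owning Fortnite player pawn into the vortex state while it is active.
; On activation it calls SetinVortex on the avatar pawn, passing the event magnitude as a parameter, and adds an enter gameplay cue to the owner.
; On ability end it switches the vortex state off and explicitly removes the cue, which is not set to auto-remove.
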(event Ability|EventActivateAbilityFromEvent
  (params (EventData :type Gameplay%20Event%20Data%20Structure%20%28by%20ref%29))
  (bind (_eventtag _instigator _target _optionalobject _optionalobject2 _contexthandle _instigatortags _targettags _eventmagnitude) (Utilities|Struct|BreakGameplayEventData EventData))
  (Vortex|SetinVortex (CallFunction|GetAvatarasFortPlayerPawn) true _eventmagnitude 0.0 0.0 0.0 0.0 true)
  (Ability|AddGameplayCueToOwner :GameplayCueTag (Variables|Default|GetGC%20Enter) :bRemoveOnAbilityEnd false))

(event Ability|EventOnEndAbility
  (params (bWasCancelled :type Boolean))
  (Vortex|SetinVortex (CallFunction|GetAvatarasFortPlayerPawn) false 0.0 0.0 0.0 0.0 0.0 false)
  (Ability|RemoveGameplayCueFromOwner (Variables|Default|GetGC%20Enter)))
